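# Control Rig for the PirouetteWeld back bling that adjusts its two wings to body pose, except when fully skydiving.
# A library function measures local rotation deltas from reference pose: clavicle_r drives the right wing, spine_05 the left.
# Clamped remaps, some Kalman-smoothed, weight additional rotations and offsets on each wing_01 bone.
# Chain harmonics nodes define an idle wing wave when not moving, but are unconnected to execution.
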ControlRigBlueprint /Game/Accessories/FORT_Backpacks/Backpack_F_MED_PirouetteWeld/Meshes/F_MED_PirouetteWeld_Pack_CtrlRig
graphs [
 ContainedGraph {nodes [
   n1 = RigVMFunctionEntryNode()
   MathTransformMakeRelative = /Script/RigVM.RigVMFunction_MathQuaternionMakeRelative(
    Global=&RigUnit_GetTransform(Space=LocalSpace, Item=@n1.Item).Transform.Rotation
    Parent=&RigUnit_GetTransform(Space=LocalSpace, bInitial=true, Item=@n1.Item).Transform.Rotation
   ) {is=struct RigVMUnitNode, template="Make Relative::Execute(in Global,in Parent,out Local)"}
   > RigVMFunctionReturnNode(OutTransform=/Script/RigVM.RigVMFunction_MathTransformToEulerTransform(Value=(Rotation=(X=0,Y=0,Z=0,W=1),Translation=(X=0,Y=0,Z=0),Scale3D=(X=1,Y=1,Z=1)), Value.Rotation=@MathTransformMakeRelative.Local) {is=struct RigVMUnitNode, template="Cast::Execute(in Value,out Result)"}.Result)
  ]}
 RigVMFunctionLibrary {nodes [GetDeltaTransformFromRefPose_FNC = RigVMCollapseNode(ExecuteContext=(), Item=(Type=Bone))]}
 RigVMModel {nodes [
   n1 = &RigVMFunction_MathBoolNot(Value=RigVMVariableNode() {variable=Is Moving}.Value)
   n2 = RigVMVariableNode() {variable=Speed}
   n3 = RigVMVariableNode() {variable=Amplitude}
   n4 = RigVMVariableNode() {variable=Frequency}
   /Script/ControlRig.RigUnit_ChainHarmonics(
    ChainRoot=wing_01_r
    DrawWorldOffset=(Rotation=(X=0,Y=0,Z=0,W=1),Translation=(X=0,Y=0,Z=0),Scale3D=(X=1,Y=1,Z=1))
    ExecutePin=()
    Pendulum=(bEnabled=false,PendulumStiffness=2,PendulumGravity=(X=0,Y=0,Z=0),PendulumBlend=1,PendulumDrag=1,PendulumMinimum=0,PendulumMaximum=0.9,PendulumEase=SineEaseIn,UnwindAxis=(X=0,Y=1,Z=0),UnwindMinimum=1,UnwindMaximum=0.1)
    Reach=(bEnabled=false,ReachTarget=(X=0,Y=0,Z=0),ReachAxis=(X=1,Y=0,Z=0),ReachMinimum=0,ReachMaximum=0)
    Speed=(X=1,Y=1,Z=1)
    Wave=(bEnabled=True,WaveFrequency=(X=1,Y=0.8,Z=-3),WaveAmplitude=(X=0,Y=0,Z=1.5),WaveOffset=(X=0,Y=1,Z=1),WaveNoise=(X=0,Y=0,Z=0),WaveMinimum=0,WaveMaximum=1)
    WaveCurve=(EditorCurveData=(Keys=((Value=0.5),(Time=0.9884,Value=0.75)),DefaultValue=340282346638528859811704183484516925440,PreInfinityExtrap=RCCE_Constant,PostInfinityExtrap=RCCE_Constant))
    bDrawDebug=True
    Wave.bEnabled=@n1.Result
    Speed.Z=@n2.Value
    Wave.WaveAmplitude.Z=@n3.Value
    Wave.WaveFrequency.Z=@n4.Value
   ) {is=struct RigVMUnitNode}
   > &RigUnit_ChainHarmonics(
    ChainRoot=wing_01_l
    DrawWorldOffset=(Rotation=(X=0,Y=0,Z=0,W=1),Translation=(X=0,Y=0,Z=0),Scale3D=(X=1,Y=1,Z=1))
    Pendulum=(bEnabled=false,PendulumStiffness=2,PendulumGravity=(X=0,Y=0,Z=0),PendulumBlend=1,PendulumDrag=1,PendulumMinimum=0,PendulumMaximum=0.9,PendulumEase=SineEaseIn,UnwindAxis=(X=0,Y=1,Z=0),UnwindMinimum=1,UnwindMaximum=0.1)
    Reach=(bEnabled=false,ReachTarget=(X=0,Y=0,Z=0),ReachAxis=(X=1,Y=0,Z=0),ReachMinimum=0,ReachMaximum=0)
    Speed=(X=1,Y=1,Z=1)
    Wave=(bEnabled=True,WaveFrequency=(X=1,Y=0.3,Z=-3),WaveAmplitude=(X=0,Y=0,Z=1.5),WaveOffset=(X=0,Y=1,Z=15),WaveNoise=(X=0,Y=0,Z=0),WaveMinimum=0,WaveMaximum=1)
    WaveCurve=(EditorCurveData=(Keys=((Value=0.5),(Time=0.9884,Value=0.75)),DefaultValue=340282346638528859811704183484516925440,PreInfinityExtrap=RCCE_Constant,PostInfinityExtrap=RCCE_Constant))
    bDrawDebug=True
    Wave.bEnabled=@n1.Result
    Speed.Z=@n2.Value
    Wave.WaveAmplitude.Z=@n3.Value
    Wave.WaveFrequency.Z=@n4.Value
   )
   RigUnit_BeginExecution = &RigUnit_BeginExecution()
   RigUnit_SequenceAggregate = &RigVMFunction_Sequence(ExecuteContext=@RigUnit_BeginExecution.ExecutePin)
   BranchNode = &RigVMFunction_ControlFlowBranch(
    Condition=&RigVMFunction_MathBoolNot(Value=RigVMVariableNode() {variable=Is Fully SkyDiving}.Value).Result
    ExecuteContext=@RigUnit_SequenceAggregate.A
   )
   RigUnit_SequenceAggregate_1_1 = &RigVMFunction_Sequence(ExecuteContext=@BranchNode.True)
   GetDeltaTransformFromRefPose_FNC_1 = RigVMFunctionReferenceNode(Item=(Type=Bone,Name="clavicle_r"), ExecuteContext=@RigUnit_SequenceAggregate_1_1.A)
   Cast = /Script/RigVM.RigVMFunction_MathQuaternionFromRotatorV2(Value=(Pitch=0,Yaw=0,Roll=15)) {is=struct RigVMUnitNode, template="Cast::Execute(in Value,out Result)"}
   > &RigUnit_ModifyTransforms(
    ItemToModify=((Item=(Type=Bone,Name="wing_01_r"),Transform=(Rotation=(X=0,Y=0,Z=0,W=1),Translation=(X=0,Y=-3,Z=0),Scale3D=(X=1,Y=1,Z=1))))
    ItemToModify.0.Transform.Rotation=@Cast.Result
    Weight=&RigVMFunction_KalmanFloat(
     BufferSize=7
     Value=&RigVMFunction_MathFloatRemap(SourceMaximum=-25, SourceMinimum=0, bClamp=true, Value=@GetDeltaTransformFromRefPose_FNC_1.OutTransform.Rotation.Yaw).Result
    ) {template="KalmanFilter::Execute(in Value,in BufferSize,out Result)"}.Result
   )
   > &RigUnit_ModifyTransforms(
    ItemToModify=((Item=(Type=Bone,Name="wing_01_r"),Transform=(Rotation=(X=0,Y=0,Z=0,W=1),Translation=(X=0,Y=1,Z=0),Scale3D=(X=1,Y=1,Z=1))))
    ItemToModify.0.Transform.Rotation=&RigVMFunction_MathQuaternionFromRotatorV2(Value=(Pitch=0,Yaw=0,Roll=-20)) {template="Cast::Execute(in Value,out Result)"}.Result
    Weight=&RigVMFunction_KalmanFloat(
     BufferSize=7
     Value=&RigVMFunction_MathFloatRemap(SourceMaximum=10, SourceMinimum=0, bClamp=true, Value=@GetDeltaTransformFromRefPose_FNC_1.OutTransform.Rotation.Yaw).Result
    ) {template="KalmanFilter::Execute(in Value,in BufferSize,out Result)"}.Result
   )
   GetDeltaTransformFromRefPose_FNC_1_1_1 = RigVMFunctionReferenceNode(Item=(Type=Bone,Name="spine_05"), ExecuteContext=@RigUnit_SequenceAggregate_1_1.B)
   > &RigUnit_ModifyTransforms(
    ItemToModify=((Item=(Type=Bone,Name="wing_01_l"),Transform=(Rotation=(X=0,Y=0,Z=0,W=1),Translation=(X=0,Y=0,Z=0),Scale3D=(X=1,Y=1,Z=1))))
    ItemToModify.0.Transform.Rotation=&RigVMFunction_MathQuaternionFromRotatorV2(Value=(Pitch=0,Yaw=0,Roll=10)) {template="Cast::Execute(in Value,out Result)"}.Result
    Weight=&RigVMFunction_MathFloatRemap(SourceMaximum=-30, SourceMinimum=-12, bClamp=true, Value=@GetDeltaTransformFromRefPose_FNC_1_1_1.OutTransform.Rotation.Yaw).Result
   )
   KalmanFilter = &RigVMFunction_KalmanFloat(
    BufferSize=7
    Value=&RigVMFunction_MathFloatRemap(SourceMaximum=14, SourceMinimum=5.3, bClamp=true, Value=@GetDeltaTransformFromRefPose_FNC_1_1_1.OutTransform.Rotation.Roll).Result
   ) {template="KalmanFilter::Execute(in Value,in BufferSize,out Result)"}
   > &RigUnit_ModifyTransforms(
    ItemToModify=((Item=(Type=Bone,Name="wing_01_l"),Transform=(Rotation=(X=0,Y=0,Z=0,W=1),Translation=(X=0,Y=-8,Z=0),Scale3D=(X=1,Y=1,Z=1))))
    Weight=@KalmanFilter.Result
   )
   > &RigUnit_ModifyTransforms(
    ItemToModify=((Item=(Type=Bone,Name="wing_01_l"),Transform=(Rotation=(X=0,Y=0,Z=0,W=1),Translation=(X=0,Y=0,Z=0),Scale3D=(X=1,Y=1,Z=1))))
    ItemToModify.0.Transform.Rotation=&RigVMFunction_MathQuaternionFromRotatorV2(Value=(Pitch=0,Yaw=0,Roll=-20)) {template="Cast::Execute(in Value,out Result)"}.Result
    Weight=&RigVMFunction_MathFloatRemap(SourceMaximum=10, SourceMinimum=0, bClamp=true, Value=@GetDeltaTransformFromRefPose_FNC_1_1_1.OutTransform.Rotation.Yaw).Result
   )
  ]}
]
hierarchy [
 root {loc=<0 0 0>, rot=<0 0 0 1>}
 body_root {parent=root, loc=<0 0 0>, rot=<0 0 0 1>}
 pelvis {parent=body_root, loc=<0 -0.5301 98.0571>, rot=<0.0127 0.707 -0.0127 -0.707>}
 spine_01 {parent=pelvis, loc=<6.1208 -0.9328 0>, rot=<0 0 0.034 -0.9994>}
 spine_02 {parent=spine_01, loc=<6.5996 -0.2095 0>, rot=<0 0 0.0238 -0.9997>}
 spine_03 {parent=spine_02, loc=<6.4914 -0.1237 0>, rot=<0 0 0.0149 -0.9999>}
 spine_04 {parent=spine_03, loc=<6.9598 0.0794 0>, rot=<0 0 -0.0132 -0.9999>}
 spine_05 {parent=spine_04, loc=<6.9987 0 0>, rot=<0 0 -0.0015 -1>}
 clavicle_l {parent=spine_05, loc=<5.4831 -0.1053 -3.2946>, rot=<0.0465 -0.7289 -0.009 -0.6829>}
 upperarm_l {parent=clavicle_l, loc=<9.8658 -1.2136 -0.4105>, rot=<0.0164 -0.3245 -0.0269 -0.9454>}
 lowerarm_l {parent=upperarm_l, loc=<26.465 -0.8345 0.259>, rot=<0.0036 0.0161 -0.2298 -0.9731>}
 clavicle_l1 {parent=clavicle_l, loc=<6.608 -6.9487 -0.8172>, rot=<0.1695 -0.6306 0.0262 -0.7569>}
 clavicle_r {parent=spine_05, loc=<5.4831 -0.1053 3.2946>, rot=<0.7289 0.0465 -0.6829 0.009>}
 upperarm_r {parent=clavicle_r, loc=<-9.8658 1.2136 0.4104>, rot=<0.0164 -0.3245 -0.0269 -0.9454>}
 lowerarm_r {parent=upperarm_r, loc=<-26.4651 0.8345 -0.259>, rot=<0.0036 0.0161 -0.2298 -0.9731>}
 clavicle_r1 {parent=clavicle_r, loc=<-6.608 6.9487 0.8173>, rot=<0.6356 0.1989 0.746 0.003>}
 neck_01 {parent=spine_05, loc=<11.3138 0.1529 0>, rot=<0 0 -0.2003 -0.9797>}
 neck_02 {parent=neck_01, loc=<4.5627 0 0>, rot=<0 0 0.0504 -0.9987>}
 head {parent=neck_02, loc=<5.3914 0 0>, rot=<0 0 0.115 -0.9934>}
 thigh_l {parent=pelvis, loc=<-1.0855 -0.2109 -8.7053>, rot=<0.0009 0.0355 0.0155 -0.9993>}
 calf_l {parent=thigh_l, loc=<-42.5284 -0.0019 -0.7297>, rot=<0.0006 -0.0161 -0.0391 -0.9991>}
 thigh_r {parent=pelvis, loc=<-1.0853 -0.2109 8.7053>, rot=<0.0355 -0.0009 0.9993 0.0155>}
 calf_r {parent=thigh_r, loc=<42.5286 0.0019 0.7297>, rot=<0.0006 -0.0161 -0.0391 -0.9991>}
 backpack_root {parent=root, loc=<0 0 0>, rot=<0 0 0 1>}
 pack {parent=backpack_root, loc=<0 -0.7316 0>, rot=<0 0 0 1>}
 wing_00_l {parent=pack, loc=<9.6681 -0.9069 14.3023>, rot=<0.1537 -0.6597 0.039 -0.7346>}
 wing_01_l {parent=wing_00_l, loc=<8.7101 2.6444 -0.3989>, rot=<0 0 0 1>}
 dyn_wing_02_l {parent=wing_01_l, loc=<8.3401 0 0>, rot=<0 0.0001 0.0098 -1>}
 dyn_wing_03_l {parent=dyn_wing_02_l, loc=<9.3046 0 0>, rot=<-0.0023 0.0081 -0.0043 -1>}
 dyn_wing_04_l {parent=dyn_wing_03_l, loc=<14.9993 0 0>, rot=<0 0.0002 0.0057 -1>}
 dyn_wing_05_l {parent=dyn_wing_04_l, loc=<17.4035 0 0>, rot=<0.0001 -0.0005 -0.0005 -1>}
 dyn_wing_06_l {parent=dyn_wing_05_l, loc=<17.8616 0 0>, rot=<-0.0001 0.0004 0.0131 -0.9999>}
 dyn_wing_07_l {parent=dyn_wing_06_l, loc=<14.629 0 0>, rot=<-0.0005 0.0013 0.0446 -0.999>}
 dyn_wing_08_l {parent=dyn_wing_07_l, loc=<11.5506 0 0>, rot=<0 0 0 1>}
 dyn_wing_end_l {parent=dyn_wing_08_l, loc=<6.185 0 0>, rot=<-0.0549 -0.0267 0.0059 -0.9981>}
 wing_00_r {parent=pack, loc=<-9.702 -0.9351 14.315>, rot=<0.0146 -0.7241 0.1822 -0.665>}
 wing_01_r {parent=wing_00_r, loc=<8.7114 2.7085 0.1149>, rot=<0 0 0 1>}
 dyn_wing_02_r {parent=wing_01_r, loc=<8.3401 0 0>, rot=<0 -0.0001 0.0098 -1>}
 dyn_wing_03_r {parent=dyn_wing_02_r, loc=<9.3046 0 0>, rot=<0.0023 -0.0081 -0.0043 -1>}
 dyn_wing_04_r {parent=dyn_wing_03_r, loc=<14.9992 0 0>, rot=<0 -0.0002 0.0057 -1>}
 dyn_wing_05_r {parent=dyn_wing_04_r, loc=<17.4035 0 0>, rot=<-0.0001 0.0002 -0.0005 -1>}
 dyn_wing_06_r {parent=dyn_wing_05_r, loc=<17.8621 0 0>, rot=<0 -0.0001 0.0131 -0.9999>}
 dyn_wing_07_r {parent=dyn_wing_06_r, loc=<14.6291 0 0>, rot=<0.0083 -0.0223 0.044 -0.9988>}
 dyn_wing_08_r {parent=dyn_wing_07_r, loc=<11.5863 0 0>, rot=<-0.0087 0.0206 0.0007 -0.9997>}
 dyn_wing_end_r {parent=dyn_wing_08_r, loc=<6.159 0 0>, rot=<-0.0758 0.0272 0.0023 -0.9967>}
 floorCOL {parent=root, loc=<0 0 0>, rot=<0 0 0 1>}
]
variables [
 Amplitude {type=double, default=0.7}
 Frequency {type=double, default=10}
 Is Fully SkyDiving {type=bool, default=False}
 Is Moving {type=bool, default=False}
 ItemToModify_0_Transform {type=FTransform, default=(Rotation=(X=0,Y=0,Z=0,W=1),Translation=(X=0,Y=0,Z=0),Scale3D=(X=1,Y=1,Z=1))}
 ItemToModify_0_Transform_Rotation {type=FQuat, default=(X=-0.342,Y=0,Z=0,W=0.9397)}
 PenGravity {type=FVector, default=(X=0,Y=0,Z=0)}
 Speed {type=double, default=1}
]
